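# Animation Blueprint for the F_MED_Skull_Brite_VDay backpack cosmetic, adding secondary motion to its zipper.
# A single Anim Dynamics node simulates the dyn_zipper bone as one constrained body, with reduced gravity, heavy damping and scaled component motion.
# Asymmetric angular limits give the zipper a wide swing on one axis and little on the others.
# The local pose feeds the simulation in component space, then converts back for output; the event graph is empty.
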AnimBlueprint /BRCosmetics/Accessories/FORT_Backpacks/F_MED_Skull_Brite_VDay/Meshes/F_MED_Skull_Brite_VDay_Pack_AnimBp
target_skeleton=&/F_MED_Skull_Brite_VDay_Pack_Skeleton
graphs [
 AnimGraph {nodes [
   n1 = Anim Dynamics - Bone: dyn_zipper(ComponentPose=Local To Component(), Alpha=1) {
    angularDampingOverride=0.98
    bGravityOverrideInSimSpace=true
    bOverrideAngularDamping=true
    bOverrideLinearDamping=true
    gravityScale=0.25
    linearDampingOverride=0.98
    alphaBoolBlend {blendOption=Linear}
    boundBone {boneName=dyn_zipper}
    componentLinearAccScale=<0.15 0.15 0.6>
    componentLinearVelScale=<0.15 0.15 0.6>
    physicsBodyDefinitions [{
      sphereCollisionRadius=0
      boundBone {boneName=dyn_zipper}
      boxExtents=<2 5 2>
      constraintSetup {angularTargetAxis=AxisZ, twistAxis=AxisZ, angularLimitsMax=<5 149.5 5>, angularLimitsMin=<-5 -15 -5>, angularTarget=<0 0 0>}
      localJointOffset=<0 0 -1>
     }]
   }
   Output Pose(Result=Component To Local(ComponentPose=@n1)) {layerGroup=None}
  ]}
 EventGraph {links [], nodes []}
]
variables []
